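; Event-triggered gameplay ability for the Season 13 WilliePete Dam 10-seconds quest, responding only when the payload target's tags pass a comparison with the quest tag variable.
; On activation it waits for the player's Launched tag, binds the pawn's landed and entered-water events, then immediately calls EndAbility.
; Either event applies the quest's gameplay effect to the owning pawn, unbinds both delegates and calls EndAbility again.
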(event Ability|EventActivateAbilityFromEvent
  (params (EventData :type Gameplay%20Event%20Data%20Structure%20%28by%20ref%29))
  (bind _owningplayerpawn (Variables|GAVars|Data|GetOwningPlayerPawn))
  (|CallParentFunction EventData)
  (Ability|Tasks|WaitGameplayTagAdd "(TagName=\"Gameplay.Action.Player.Launched\")"
    (:Added
      (EventDispatchers|AssignOnPawnLanded _owningplayerpawn (EventDispatchers|CreateEvent :!event Custom|Landed))
      (EventDispatchers|AssignOnPawnEnterWater _owningplayerpawn (EventDispatchers|CreateEvent :!event Custom|Enter%20Water))
      (Ability|EndAbility))))

(event Custom|Enter%20Water
  (params (FortPlayerPawn :type Fort%20Player%20Pawn%20Object%20Reference))
  (CallFunction|ApplyGE)
  (CallFunction|Unbind))

(event Custom|Landed
  (params (Hit :type Hit%20Result%20Structure%20%28by%20ref%29))
  (CallFunction|ApplyGE)
  (CallFunction|Unbind))

(event Custom|Unbind
  (bind _owningplayerpawn (Variables|GAVars|Data|GetOwningPlayerPawn))
  (EventDispatchers|UnbindEventfromOnPawnLanded _owningplayerpawn (delegate Custom|Landed))
  (EventDispatchers|UnbindEventfromOnPawnEnterWater _owningplayerpawn (delegate Custom|Enter%20Water))
  (Ability|EndAbility))

(event Custom|Apply%20GE
  (GameplayEffects|ApplyGameplayEffectToSelf :self (Ability|GetAbilitySystemComponent (Variables|GAVars|Data|GetOwningPlayerPawn)) :GameplayEffectClass "/Game/Athena/Items/Quests/QuestAbilities/Quest_S13_WilliePete_Dam_10seconds/GE_Quest_S13_WilliePete_Dam_10seconds.GE_Quest_S13_WilliePete_Dam_10seconds_C" :Level 0.0))

(fn ShouldAbilityRespondToEvent
  (params (ActorInfo :type Gameplay%20Ability%20Actor%20Info%20Structure) (Payload_EventTag :type Gameplay%20Tag%20Structure) (Payload_Instigator :type Actor%20Object%20Reference) (Payload_Target :type Actor%20Object%20Reference) (Payload_OptionalObject :type Object%20Reference) (Payload_OptionalObject2 :type Object%20Reference) (Payload_ContextHandle :type Gameplay%20Effect%20Context%20Handle%20Structure) (Payload_InstigatorTags :type Gameplay%20Tag%20Container%20Structure) (Payload_TargetTags :type Gameplay%20Tag%20Container%20Structure) (Payload_EventMagnitude :type Float%20%28single-precision%29) (Payload_TargetData :type Gameplay%20Ability%20Target%20Data%20Handle%20Structure) (Payload_InstancedEventData :type Instanced%20Struct%20Structure))
  (return (GameplayTags|TagContainer|CompareTagAssetInterfacetoOtherTags (Variables|Default|GetT_WilliePete_Dam) (Ability|GetAbilitySystemComponent Payload_Target))))
